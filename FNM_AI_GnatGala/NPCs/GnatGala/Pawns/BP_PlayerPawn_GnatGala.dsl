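; Boss NPC pawn for GnatGala ("Jason"): plays proximity audio and music, with an optional server-controlled map marker.
; A looping timer registers a nearby-actor perception listener, caching found actors and removing Marked Target from those exiting.
; A second timer loops over cached actors, adding or removing a self speed boost per actor's distance to aid short-range melee.
; Death fades VO audio, clears Marked Target from cached actors and deactivates the marker.
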(event Custom|RemoveEffectTimer
  ;[1 1 1] TEST - removes Marked Target effect
  (for _array_element (Variables|Default|GetFound%20Actor%20Array)
    (GameplayEffects|ApplyGameplayEffectToTarget :self (Ability|GetAbilitySystemComponent 0) :GameplayEffectClass (Variables|Default|GetMark%20Target%20Removal%20Effect) :Target (Ability|GetAbilitySystemComponent _array_element) :Level 0.0)))

(event EventBeginPlay
  (bind _voaudiocomponent (Variables|Default|GetVOAudioComponent))
  ;[1 1 1] Begin Play
  (|CallParentFunction)
  (Audio|SpawnSoundAttached :Sound (Variables|Default|GetProximity_Sound) :AttachToComponent _voaudiocomponent :AttachPointName "None" :LocationType "KeepRelativeOffset" :bStopWhenAttachedToDestroyed true :VolumeMultiplier 1.0 :PitchMultiplier 1.0 :StartTime 0.0 :bAutoDestroy true)
  (Audio|SpawnSoundAttached :Sound (Variables|Default|GetIP_Music) :AttachToComponent _voaudiocomponent :AttachPointName "None" :LocationType "KeepRelativeOffset" :bStopWhenAttachedToDestroyed true :VolumeMultiplier 1.0 :PitchMultiplier 1.0 :StartTime 0.0 :bAutoDestroy true)
  (Utilities|Time|SetTimerbyEvent (delegate Custom|FindTargets) 2.0 true)
  ;[1 1 1] Increases Jason's Speed when a Target is close by, then decreases that speed when the target is out of range. This is to help give coverage on Jason's extremely short-ranged melee attack.
  (Utilities|Time|SetTimerbyEvent (delegate Custom|ChangeJasonsSpeed) 0.5 true)
  ;[1 1 1] Boss Map Marker
  (Networking|SwitchHasAuthority
    (:Authority
      (Class|FortAsyncActionCurrentPlaylistReady|PlaylistDataReadyAsync
        (:OnReady
          (if (ScalableFloat|GetValueasBool (Variables|Default|GetShowBossMapMarker))
            (Components|Activation|Activate (Variables|Default|GetAthenaSpecialActor) false)))))))

(event Custom|ChangeJasonsSpeed
  ;[1 1 1] Increases Jason's Speed when a Target is close by, then decreases that speed when the target is out of range. This is to help give coverage on Jason's extremely short-ranged melee attack.
  (for _array_element (Variables|Default|GetFound%20Actor%20Array)
    (if (<= (Transformation|GetDistanceTo _array_element self) (Variables|Default|GetMax%20Range%20from%20Target%20to%20Incrase%20Speed))
      (GameplayEffects|ApplyGameplayEffectToTarget :self (Ability|GetAbilitySystemComponent self) :GameplayEffectClass (Variables|Default|GetApplySpeedIncrease) :Target (Ability|GetAbilitySystemComponent self) :Level 0.0)
      (else
        (GameplayEffects|ApplyGameplayEffectToTarget :self (Ability|GetAbilitySystemComponent self) :GameplayEffectClass (Variables|Default|GetRemoveSpeedIncrease) :Target (Ability|GetAbilitySystemComponent self) :Level 0.0)))))

(event Custom|RemoveOldActorsScreenPresentation)

(event Custom|FindTargets
  (if (not (<= (FortPawn|GetHealth) 0))
    (Query|BPRegisterListener (Variables|Default|GetFortNearbyActorsPerception) (Utilities|Struct|MakeFortNearbyActorsPerceptionConfiguration 2.0 3 0 2e+04) (EventDispatchers|CreateEvent :!event Custom|FoundActors))))

(event Custom|FoundActors
  (params (AllFoundActors :type Array%20of%20Actor%20Object%20References) (EnteringActors :type Array%20of%20Actor%20Object%20References) (ExitingActors :type Array%20of%20Actor%20Object%20References))
  ;[1 1 1] Setting an array of all actors found by GnatGala
  (Variables|Default|SetFound%20Actor%20Array AllFoundActors)
  ;[1 1 1] Removing VFX and Audio when Players are no longer percievable by GnatGala.
  (for _array_element ExitingActors
    (GameplayEffects|ApplyGameplayEffectToTarget :self (Ability|GetAbilitySystemComponent _array_element) :GameplayEffectClass (Variables|Default|GetMark%20Target%20Removal%20Effect) :Target (Ability|GetAbilitySystemComponent _array_element) :Level 0.0)))

(event EventOnDeathServer
  (params (Damage :type Float%20%28single-precision%29) (DamageTags :type Gameplay%20Tag%20Container%20Structure%20%28by%20ref%29) (Momentum :type Vector :default "0, 0, 0") (HitInfo :type Hit%20Result%20Structure%20%28by%20ref%29) (InstigatedBy :type Controller%20Object%20Reference) (DamageCauser :type Actor%20Object%20Reference) (EffectContext :type Gameplay%20Effect%20Context%20Handle%20Structure))
  (|CallParentFunction Damage DamageTags Momentum HitInfo InstigatedBy DamageCauser EffectContext)
  ;[1 1 1] Making sure that Marked Target gets removed from everyone if Jason dies, no matter what.
  (Audio|Components|Audio|FadeOut (Variables|Default|GetVOAudioComponent) 3.0 0.0 "Linear" :!class /Script/Engine.AudioComponent)
  (for _array_element (Variables|Default|GetFound%20Actor%20Array)
    (GameplayEffects|ApplyGameplayEffectToTarget :self (Ability|GetAbilitySystemComponent _array_element) :GameplayEffectClass (Variables|Default|GetMark%20Target%20Removal%20Effect) :Target (Ability|GetAbilitySystemComponent _array_element) :Level 0.0)))

(event EventOnDied_Event
  (params (DamagedActor :type Actor%20Object%20Reference) (Damage :type Float%20%28single-precision%29) (InstigatedBy :type Controller%20Object%20Reference) (DamageCauser :type Actor%20Object%20Reference) (HitLocation :type Vector :default "0, 0, 0") (FHitComponent :type Primitive%20Component%20Object%20Reference) (BoneName :type Name :default "None") (Momentum :type Vector :default "0, 0, 0"))
  ;[1 1 1] On Died
  (|CallParentFunction DamagedActor Damage InstigatedBy DamageCauser HitLocation FHitComponent BoneName Momentum)
  (Networking|SwitchHasAuthority
    (:Authority
      (Components|Activation|Deactivate (Variables|Default|GetAthenaSpecialActor)))))

(fn ConstructionScript
  (|CallParentFunction))
